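# Niagara trail effect for the CanineCronut drumstick pickaxe cosmetic, built from two short-lived ribbon emitters.
# Trail_Instrument_Color draws an anim trail between bones or sockets, while Ribbon_CameraFacing001 spawns ribbon particles continuously from skeletal mesh locations.
# Both taper width and fade over life, tinting by a System.TOD color, varying alpha with time of day and dimming in the front end.
# The second ribbon also animates dynamic material parameters.
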NiagaraSystem /SparksCosmetics/Instrument/DrumStick/CanineCronut/Effects/NS_Pickaxe_Drumstick_CanineCronut_Trail
dependencies {
 usedDataInterfaces [&NiagaraDataInterfaceCurve, &NiagaraDataInterfaceColorCurve]
 usedDynamicInputs [&ColorFromCurve]
 usedModules [
  &self:SystemSpawnScript.SystemScriptSource.SystemScriptGraph.NiagaraNodeAssignment_0.SetVariables_#1
  &SystemState
 ]
}
emitters {
 Ribbon_CameraFacing001 {
  emitterUpdateScript [
   &EmitterState(Life Cycle Mode=::NewEnumerator1, Loop Behavior=::NewEnumerator1, Loop Delay=0.005, Loop Duration=1, UseLoopDelay=true)
   &SpawnRate(SpawnRate=45, Use Spawn Probability=false)
  ]
  particleSpawnScript [
   &InitializeRibbon(
    Color=<rgba 0.1969 0.3073 0.2573 0.75>
    Lifetime=0.1
    Module.Write Color=true
    Module.Write Lifetime=true
    Module.Write Mass=false
    Module.Write Material Random=true
    Module.Write Position=false
    Module.Write RibbonTwist=false
    Module.Write RibbonWidth=true
   )
   &SkeletalMeshLocation(
    Inherit Velocity=false
    MaskSpawnGroup=false
    Mesh Sampling Type=::NewEnumerator5
    Sampled Position Offset Coordinate Space:ENiagaraMeshTransforms=::NewEnumerator3
    Skeletal Mesh=$User.SkeletalMesh
    Socket Sampling Mode=::NewEnumerator5
   )
  ]
  particleUpdateScript [
   &ParticleState()
   &SolveForcesAndVelocity(Clamp Velocity=false, Limit Acceleration=false)
   &ScaleColor(
    Scale Alpha=&Multiply_Float(
     A=&FloatFromCurve(
      FloatCurve=DataInterface() {propertyValues {Curve {keys [
          {interpMode=RCIM_Cubic, leaveTangent=-3.5583, tangentMode=RCTM_Break, time=0, value=1}
          <at 0.3 0.15>
          {arriveTangent=-0.2139, interpMode=RCIM_Cubic, tangentMode=RCTM_Break, time=1, value=0}
         ]}}}
      Scale Curve=&FloatFromCurve(
       FloatCurve=DataInterface() {propertyValues {Curve {keys [<at 0 0.75>,<at 0.2 1>,<at 0.75 1>,<at 1 0.75>]}}}
       CurveIndex=Linked() {linkedVariable {name=NPC.FortniteNPC.TimeOfDay_0-1, type=NiagaraFloat}}
      )
     )
     B=&FN_FrontEndFloatModifier(Gameplay=1, Front End=0.4)
    )
    Scale RGB=&MakeVectorFromLinearColorRGB(Color=$System.TOD)
    ScaleA=true
    ScaleRGB=true
   ) {module=ScaleColor002}
   &ScaleRibbonWidth(Ribbon Width Scale=&FloatFromCurve(FloatCurve=DataInterface() {propertyValues {Curve {
        defaultValue=1
        keys [{tangentMode=RCTM_User, time=0, value=1}, {arriveTangent=-1, interpMode=RCIM_Cubic, leaveTangent=-1, time=0.3, value=1}, <at 1 0 RCIM_Cubic>]
       }}}))
   &DynamicMaterialParameters(
    Index 0 Param 1=&FloatFromCurve(FloatCurve=DataInterface() {propertyValues {Curve {keys [
         {arriveTangent=0.2015, interpMode=RCIM_Cubic, leaveTangent=0.2015, tangentMode=RCTM_User, time=0, value=1.5}
         {arriveTangent=-1.5941, interpMode=RCIM_Cubic, leaveTangent=-1.5941, tangentMode=RCTM_User, time=0.3, value=1}
         {interpMode=RCIM_Cubic, tangentMode=RCTM_User, time=1, value=0}
        ]}}})
    Index 0 Param 3=-0.8
    Index 0 Param 4=&FloatFromCurve(FloatCurve=DataInterface() {propertyValues {Curve {keys [
         {interpMode=RCIM_Cubic, tangentMode=RCTM_User, time=0, value=1}
         {arriveTangent=-0.6219, interpMode=RCIM_Cubic, leaveTangent=-0.6219, tangentMode=RCTM_User, time=1, value=0}
        ]}}})
    Module.Param0WriteEnabled=true
    Module.Param1WriteEnabled=true
    Module.Param2WriteEnabled=true
    Module.Param3WriteEnabled=true
   )
  ]
  rendererClasses [&NiagaraRibbonRendererProperties]
  renderers [{rendererClass=&NiagaraRibbonRendererProperties}]
 }
 Trail_Instrument_Color {
  emitterUpdateScript [
   &EmitterState(Life Cycle Mode=::NewEnumerator1, Loop Behavior=::NewEnumerator1, Loop Duration=0.22, UseLoopDelay=true)
   &FN_SpawnAnimTrail_v2()
  ]
  particleSpawnScript [
   /Game/Effects/Niagara/Modules/AnimTrail/FN_InitializeAnimTrail/FN_InitializeAnimTrailBoneIndecies(First Bone or Socket Index=1)
   /Game/Effects/Niagara/Modules/AnimTrail/FN_InitializeAnimTrail/FN_InitializeAnimTrail(AnimTrailWidth=1, Color=<rgba 1 1 1 0.6>, Lifetime=0.12)
   &SystemLocation(MaskSpawnGroup=false)
  ]
  particleUpdateScript [
   &ScaleRibbonWidth(Ribbon Width Scale=&FloatFromCurve(FloatCurve=DataInterface() {propertyValues {Curve {keys [
         {arriveTangent=-1.2579, interpMode=RCIM_Cubic, time=0.1, value=1}
         {arriveTangent=-0.4927, interpMode=RCIM_Cubic, leaveTangent=-0.4927, tangentMode=RCTM_User, time=1, value=0}
        ]}}}))
   &ScaleColor(
    Scale Alpha=&Multiply_Float(
     A=&FloatFromCurve(
      FloatCurve=DataInterface() {propertyValues {Curve {keys [
          {interpMode=RCIM_Cubic, leaveTangent=-2.3646, tangentMode=RCTM_Break, time=0, value=1}
          <at 0.3 0.1>
          {arriveTangent=-0.2139, interpMode=RCIM_Cubic, tangentMode=RCTM_Break, time=1, value=0}
         ]}}}
      Scale Curve=&FloatFromCurve(
       FloatCurve=DataInterface() {propertyValues {Curve {keys [
           <at 0 0.8 RCIM_Cubic>
           {arriveTangent=0.3316, interpMode=RCIM_Cubic, leaveTangent=0.3316, time=0.4044, value=1}
           {arriveTangent=-0.3358, interpMode=RCIM_Cubic, leaveTangent=-0.3358, time=0.6032, value=1}
           <at 1 0.8 RCIM_Cubic>
          ]}}}
       CurveIndex=Linked() {linkedVariable {name=NPC.FortniteNPC.TimeOfDay_0-1, type=NiagaraFloat}}
      )
     )
     B=&FN_FrontEndFloatModifier(Gameplay=1, Front End=0.4)
    )
    Scale RGB=&MakeVectorFromLinearColorRGB(Color=$System.TOD)
    ScaleA=true
    ScaleRGB=true
   ) {module=ScaleColor002}
   &FN_UpdateAnimTrail_v2()
  ]
  rendererClasses [&NiagaraRibbonRendererProperties]
  renderers [{rendererClass=&NiagaraRibbonRendererProperties}]
 }
}
summary {
 systemName=NS_Pickaxe_Drumstick_CanineCronut_Trail
 emitters [
  {emitterName=Trail_Instrument_Color, rendererClasses [&NiagaraRibbonRendererProperties]}
  {emitterName=Ribbon_CameraFacing001, rendererClasses [&NiagaraRibbonRendererProperties]}
 ]
 userVariables [
  User.bIsFrontEnd {type=NiagaraBool}
  User.bIsFrontEndPreview {type=NiagaraBool}
  User.RibbonIDOffset {type=NiagaraInt32}
  User.SkeletalMesh {
   type=NiagaraDataInterfaceSkeletalMesh
   defaultValue {
    struct=NiagaraExt_VariableValue_DataInterface
    value {dataInterface=&self:NiagaraDataInterfaceSkeletalMesh_0}
   }
  }
  User.TrailColor {type=LinearColor, description="", defaultValue {struct=LinearColor, value=<rgba 1 1 1 0.25>}}
  User.TrailSecondaryColor {type=LinearColor, description="", defaultValue {struct=LinearColor, value=<rgba 10 10 10 0.25>}}
 ]
}
