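# Animation Blueprint for the CanineCronut drumstick cosmetic that adds physical sway to its dangling charm and tassel.
# Three Anim Dynamics nodes drive dyn_charm_01, dyn_charm_02 and dyn_charm_tassle with gravity, damping and angular limits.
# The chain appears twice in series, gated by bIsInFrontEnd and its inverse, so front end and gameplay use separate tuning.
# The non-front-end set adds sim-space settings and looser first-bone acceleration; the event graph is empty.
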AnimBlueprint /SparksCosmetics/Instrument/DrumStick/CanineCronut/Meshes/CanineCronut_DrumStick_AnimBP
target_skeleton=&/CanineCronut_DrumStick_Skeleton
graphs [
 AnimGraph {nodes [
   n1 = Get bIsInFrontEnd()
   n2 = Anim Dynamics - Bone: dyn_charm_01!(ComponentPose=Local To Component(LocalPose=Input Pose()), bAlphaBoolEnabled:bool=@n1, ExternalForce="0,0,0") -> (Pose) {
    alphaInputType=Bool
    angularDampingOverride=0.99
    bOverrideAngularDamping=true
    bUseGravityOverride=true
    alphaBoolBlend {blendOption=Linear}
    boundBone {boneName=dyn_charm_01}
    componentAppliedLinearAccClamp=<200 200 200>
    componentLinearAccScale=<0.5 0.5 0.5>
    componentLinearVelScale=<0.5 0.5 0.5>
    gravityOverride {z=-980}
    physicsBodyDefinitions [{
      boundBone {boneName=dyn_charm_01}
      boxExtents=<10 10 10>
      constraintSetup {angularTargetAxis=AxisY, twistAxis=AxisZ, angularLimitsMax=<0 45 0>, angularLimitsMin=<0 -60 0>, angularTarget=<0 0 0>}
      localJointOffset=<1 0 0>
     }]
   }
   n3 = Anim Dynamics - Bone: dyn_charm_02!(ComponentPose=@n2, bAlphaBoolEnabled:bool=@n1, GravityOverride:Vector="0,0,-980", ExternalForce="0,0,0") -> (Pose) {
    alphaInputType=Bool
    angularDampingOverride=0.99
    bOverrideAngularDamping=true
    bUseGravityOverride=true
    linearDampingOverride=0.9
    alphaBoolBlend {blendOption=Linear}
    boundBone {boneName=dyn_charm_02}
    componentAppliedLinearAccClamp=<200 200 200>
    componentLinearAccScale=<0.6 0.6 0.6>
    componentLinearVelScale=<0.5 0.5 0.5>
    physicsBodyDefinitions [{
      boundBone {boneName=dyn_charm_02}
      boxExtents=<6 6 6>
      constraintSetup {angularLimitsMax=<15 20 45>, angularLimitsMin=<-15 -45 -45>, angularTarget=<0 0 0>, linearAxesMin=<-0.25 -0.25 -0.25>}
      localJointOffset=<1 0 0>
     }]
   }
   n4 = Anim Dynamics - Bone: dyn_charm_tassle!(ComponentPose=@n3, bAlphaBoolEnabled:bool=@n1, GravityOverride:Vector="0,0,-980", ExternalForce="0,0,0") -> (Pose) {
    alphaInputType=Bool
    angularDampingOverride=0.99
    bOverrideAngularDamping=true
    bUseGravityOverride=true
    linearDampingOverride=0.9
    alphaBoolBlend {blendOption=Linear}
    boundBone {boneName=dyn_charm_tassle}
    componentAppliedLinearAccClamp=<200 200 200>
    componentLinearAccScale=<0.6 0.6 0.6>
    componentLinearVelScale=<0.5 0.5 0.5>
    physicsBodyDefinitions [{
      boundBone {boneName=dyn_charm_tassle}
      boxExtents=<6 6 6>
      constraintSetup {angularLimitsMax=<15 35 45>, angularLimitsMin=<-15 -60 -45>, angularTarget=<0 0 0>, linearAxesMin=<-0.25 -0.25 -0.25>}
      localJointOffset=<1 0 0>
     }]
   }
   n5 = NOT Boolean(A=@n1)
   n6 = Anim Dynamics - Bone: dyn_charm_01(ComponentPose=@n4, bAlphaBoolEnabled=@n5, ExternalForce="0,0,0") {
    alphaInputType=Bool
    angularDampingOverride=0.99
    bOverrideAngularDamping=true
    bUseGravityOverride=true
    alphaBoolBlend {blendOption=Linear}
    boundBone {boneName=dyn_charm_01}
    componentAppliedLinearAccClamp=<1000 1000 1000>
    componentLinearAccScale=<0.8 0.8 0.8>
    componentLinearVelScale=<0.5 0.5 0.5>
    gravityOverride {z=-980}
    physicsBodyDefinitions [{
      boundBone {boneName=dyn_charm_01}
      boxExtents=<10 10 10>
      constraintSetup {angularTargetAxis=AxisY, twistAxis=AxisZ, angularLimitsMax=<0 45 0>, angularLimitsMin=<0 -60 0>, angularTarget=<0 0 0>}
      localJointOffset=<1 0 0>
     }]
    simSpaceSettings {maxAngularAcceleration=100, maxAngularVelocity=10, simSpaceAngularAlpha=1}
   }
   n7 = Anim Dynamics - Bone: dyn_charm_02(ComponentPose=@n6, bAlphaBoolEnabled=@n5, GravityOverride="0,0,-980", ExternalForce="0,0,0") {
    alphaInputType=Bool
    angularDampingOverride=0.99
    bOverrideAngularDamping=true
    bUseGravityOverride=true
    linearDampingOverride=0.9
    alphaBoolBlend {blendOption=Linear}
    boundBone {boneName=dyn_charm_02}
    componentAppliedLinearAccClamp=<200 200 200>
    componentLinearAccScale=<0.6 0.6 0.6>
    componentLinearVelScale=<0.5 0.5 0.5>
    physicsBodyDefinitions [{
      boundBone {boneName=dyn_charm_02}
      boxExtents=<6 6 6>
      constraintSetup {angularLimitsMax=<15 20 45>, angularLimitsMin=<-15 -45 -45>, angularTarget=<0 0 0>, linearAxesMin=<-0.25 -0.25 -0.25>}
      localJointOffset=<1 0 0>
     }]
    simSpaceSettings {maxAngularAcceleration=10, maxAngularVelocity=6}
   }
   n8 = Anim Dynamics - Bone: dyn_charm_tassle(ComponentPose=@n7, bAlphaBoolEnabled=@n5, GravityOverride="0,0,-980", ExternalForce="0,0,0") {
    alphaInputType=Bool
    angularDampingOverride=0.99
    bOverrideAngularDamping=true
    bUseGravityOverride=true
    linearDampingOverride=0.9
    alphaBoolBlend {blendOption=Linear}
    boundBone {boneName=dyn_charm_tassle}
    componentAppliedLinearAccClamp=<200 200 200>
    componentLinearAccScale=<0.6 0.6 0.6>
    componentLinearVelScale=<0.5 0.5 0.5>
    physicsBodyDefinitions [{
      boundBone {boneName=dyn_charm_tassle}
      boxExtents=<6 6 6>
      constraintSetup {angularLimitsMax=<15 35 45>, angularLimitsMin=<-15 -60 -45>, angularTarget=<0 0 0>, linearAxesMin=<-0.25 -0.25 -0.25>}
      localJointOffset=<1 0 0>
     }]
    simSpaceSettings {maxAngularAcceleration=100, maxAngularVelocity=10, simSpaceAngularAlpha=1}
   }
   Output Pose(Result=Component To Local(ComponentPose=@n8))
  ]}
 EventGraph {nodes [
   TryGetPawnOwner()
   Event BlueprintUpdateAnimation()
  ]}
]
variables []
