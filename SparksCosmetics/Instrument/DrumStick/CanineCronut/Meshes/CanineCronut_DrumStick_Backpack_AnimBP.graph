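# Animation Blueprint for the CanineCronut drumstick back bling that adds secondary physics motion to its dangling charms and tassels.
# Six chained Anim Dynamics nodes simulate the original and duplicate charm bone chains over the input pose, using angular limits, damping and gravity overrides.
# The update event sets a GravityOverride variable from skydiving and parachute states, but the dynamics nodes show fixed gravity values rather than reading it.
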AnimBlueprint /SparksCosmetics/Instrument/DrumStick/CanineCronut/Meshes/CanineCronut_DrumStick_Backpack_AnimBP
target_skeleton=&/CanineCronut_DrumStick_Backpack_Skeleton
graphs [
 AnimGraph {nodes [
   n1 = Anim Dynamics - Bone: dyn_charm_01!(ComponentPose=Local To Component(LocalPose=Input Pose()), Alpha=1, GravityOverride:Vector="0,0,-980", ExternalForce="0,0,0") -> (Pose) {
    angularDampingOverride=0.99
    bOverrideAngularDamping=true
    bUseGravityOverride=true
    alphaBoolBlend {blendOption=Linear}
    boundBone {boneName=dyn_charm_01}
    componentAppliedLinearAccClamp=<200 200 200>
    componentLinearAccScale=<0.5 0.5 0.5>
    componentLinearVelScale=<0.5 0.5 0.5>
    gravityOverride {z=-980}
    physicsBodyDefinitions [{
      boundBone {boneName=dyn_charm_01}
      boxExtents=<10 10 10>
      constraintSetup {angularTargetAxis=AxisY, twistAxis=AxisZ, angularLimitsMax=<0 45 -5>, angularLimitsMin=<0 -60 -45>, angularTarget=<0 0 0>}
      localJointOffset=<1 0 0>
     }]
   }
   n2 = Anim Dynamics - Bone: dyn_charm_02!(ComponentPose=@n1, Alpha=1, GravityOverride:Vector="0,0,-980", ExternalForce="0,0,0") -> (Pose) {
    angularDampingOverride=0.99
    bOverrideAngularDamping=true
    bUseGravityOverride=true
    linearDampingOverride=0.9
    alphaBoolBlend {blendOption=Linear}
    boundBone {boneName=dyn_charm_02}
    componentAppliedLinearAccClamp=<200 200 200>
    componentLinearAccScale=<0.6 0.6 0.6>
    componentLinearVelScale=<0.5 0.5 0.5>
    physicsBodyDefinitions [{
      boundBone {boneName=dyn_charm_02}
      boxExtents=<6 6 6>
      constraintSetup {angularLimitsMax=<15 20 -5>, angularLimitsMin=<-15 -45 -45>, angularTarget=<0 0 0>, linearAxesMin=<-0.25 -0.25 -0.25>}
      localJointOffset=<1 0 0>
     }]
   }
   n3 = Anim Dynamics - Bone: dyn_charm_tassle!(ComponentPose=@n2, Alpha=1, GravityOverride:Vector="0,0,-980", ExternalForce="0,0,0") -> (Pose) {
    angularDampingOverride=0.99
    bOverrideAngularDamping=true
    bUseGravityOverride=true
    linearDampingOverride=0.9
    alphaBoolBlend {blendOption=Linear}
    boundBone {boneName=dyn_charm_tassle}
    componentAppliedLinearAccClamp=<200 200 200>
    componentLinearAccScale=<0.6 0.6 0.6>
    componentLinearVelScale=<0.5 0.5 0.5>
    physicsBodyDefinitions [{
      boundBone {boneName=dyn_charm_tassle}
      boxExtents=<6 6 6>
      constraintSetup {angularLimitsMax=<15 35 -5>, angularLimitsMin=<-15 -60 -45>, angularTarget=<0 0 0>, linearAxesMin=<-0.25 -0.25 -0.25>}
      localJointOffset=<1 0 0>
     }]
   }
   n4 = Anim Dynamics - Bone: Dupe_dyn_charm_01!(ComponentPose=@n3, Alpha=1, GravityOverride:Vector="0,0,-980", ExternalForce="0,0,0") -> (Pose) {
    angularDampingOverride=0.99
    bOverrideAngularDamping=true
    bUseGravityOverride=true
    alphaBoolBlend {blendOption=Linear}
    boundBone {boneName=Dupe_dyn_charm_01}
    componentAppliedLinearAccClamp=<200 200 200>
    componentLinearAccScale=<0.5 0.5 0.5>
    componentLinearVelScale=<0.5 0.5 0.5>
    gravityOverride {z=-980}
    physicsBodyDefinitions [{
      boundBone {boneName=Dupe_dyn_charm_01}
      boxExtents=<10 10 10>
      constraintSetup {angularTargetAxis=AxisY, twistAxis=AxisZ, angularLimitsMax=<0 45 0>, angularLimitsMin=<0 -60 -15>, angularTarget=<0 0 0>}
      localJointOffset=<1 0 0>
     }]
   }
   n5 = Anim Dynamics - Bone: Dupe_dyn_charm_02!(ComponentPose=@n4, Alpha=1, GravityOverride:Vector="0,0,-980", ExternalForce="0,0,0") -> (Pose) {
    angularDampingOverride=0.99
    bOverrideAngularDamping=true
    bUseGravityOverride=true
    linearDampingOverride=0.9
    alphaBoolBlend {blendOption=Linear}
    boundBone {boneName=Dupe_dyn_charm_02}
    componentAppliedLinearAccClamp=<200 200 200>
    componentLinearAccScale=<0.6 0.6 0.6>
    componentLinearVelScale=<0.5 0.5 0.5>
    physicsBodyDefinitions [{
      boundBone {boneName=Dupe_dyn_charm_02}
      boxExtents=<6 6 6>
      constraintSetup {angularLimitsMax=<15 20 45>, angularLimitsMin=<-15 -45 -15>, angularTarget=<0 0 0>}
      localJointOffset=<1 0 0>
     }]
   }
   n6 = Anim Dynamics - Bone: Dupe_dyn_charm_tassle!(ComponentPose=@n5, Alpha=1, GravityOverride:Vector="0,0,-980", ExternalForce="0,0,0") -> (Pose) {
    angularDampingOverride=0.99
    bOverrideAngularDamping=true
    bUseGravityOverride=true
    linearDampingOverride=0.9
    alphaBoolBlend {blendOption=Linear}
    boundBone {boneName=Dupe_dyn_charm_tassle}
    componentAppliedLinearAccClamp=<200 200 200>
    componentLinearAccScale=<0.6 0.6 0.6>
    componentLinearVelScale=<0.5 0.5 0.5>
    physicsBodyDefinitions [{
      boundBone {boneName=Dupe_dyn_charm_tassle}
      boxExtents=<6 6 6>
      constraintSetup {angularLimitsMax=<15 35 45>, angularLimitsMin=<-15 -60 -15>, angularTarget=<0 0 0>}
      localJointOffset=<1 0 0>
     }]
   }
   Output Pose(Result=Component To Local(ComponentPose=@n6))
  ]}
 EventGraph {nodes [
   Event BlueprintUpdateAnimation()
   > n1 = Branch(Condition=Get bIsSkyDiving())
   > n2 = Branch(Condition=Get bIsParachuteOpened())
   > Set GravityOverride(GravityOverride="0,0,-250")
   Set GravityOverride(execute=@n2.else, GravityOverride="0,0,200")
   Set GravityOverride(execute=@n1.else, GravityOverride="0,0,-980")
   TryGetPawnOwner()
  ]}
]
variables [GravityOverride {type=Vector}]
